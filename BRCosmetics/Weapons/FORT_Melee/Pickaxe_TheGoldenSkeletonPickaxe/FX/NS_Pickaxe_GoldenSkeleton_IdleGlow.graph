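# Niagara idle glow for the Golden Skeleton pickaxe, built from two orange-gold emitters that follow a shared socket reader.
# SplineGlow burst-spawns a persistent mesh that fades in by system age, using a separate, later curve for front end.
# Decal_Packed continuously spawns short-lived, fading sprites, split into four partitions to cover distinct socket-relative regions.
# The partition index also selects per-region lifetime, size and camera offset through merge-helper switches.
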NiagaraSystem /BRCosmetics/Weapons/FORT_Melee/Pickaxe_TheGoldenSkeletonPickaxe/FX/NS_Pickaxe_GoldenSkeleton_IdleGlow
dependencies {
 usedDataInterfaces [&NiagaraDataInterfaceCurve, &NiagaraDataInterfaceSocketReader]
 usedDynamicInputs []
 usedModules [
  &self:SystemSpawnScript.SystemScriptSource.SystemScriptGraph.NiagaraNodeAssignment_0.SetVariables_#1
  &SystemState
 ]
}
emitters {
 Decal_Packed {
  emitterUpdateScript [
   &EmitterState(Inactive Response=::NewEnumerator1, Life Cycle Mode=::NewEnumerator1, Loop Duration=1, UseLoopDelay=false)
   &SpawnRate(SpawnRate=15, Use Spawn Probability=false)
  ]
  particleSpawnScript [
   &PartitionParticles(Number Of Partitions=4)
   &InitializeParticle(
    Lifetime=&MergeHelper_FloatSwitch(
     0=&RandomRangeFloat(Minimum=1.8, Maximum=1.5, Evaluation Type=::NewEnumerator1)
     1=&RandomRangeFloat(Minimum=1.8, Maximum=1.5, Evaluation Type=::NewEnumerator1)
     2=1.6
     3=?
     Index=$Particles.PartitionParticles.Partition
     Num=::NewEnumerator1
    )
    Sprite Rotation Angle=&RandomRangeFloat(Minimum=0.5, Maximum=-0.5, Evaluation Type=::NewEnumerator1)
    Sprite Rotation Mode=::NewEnumerator1
    Sprite Size=&Vector2DFromFloat(Value=&MergeHelper_FloatSwitch(
      0=&RandomRangeFloat(Minimum=17, Maximum=15, Evaluation Type=::NewEnumerator1)
      1=&RandomRangeFloat(Minimum=17, Maximum=15, Evaluation Type=::NewEnumerator1)
      2=&RandomRangeFloat(Minimum=17, Maximum=12, Evaluation Type=::NewEnumerator1)
      3=?
      Index=$Particles.PartitionParticles.Partition
      Num=::NewEnumerator1
     ))
    Sprite Size Mode=::NewEnumerator4
    UsePositionOffset=false
    Write Material Random=true
   )
  ]
  particleUpdateScript [
   &ParticleState()
   &SocketLocation(
    SamplePosOffset=true
    Sampled Position Offset=&MergeHelper_VectorSwitch(
     0=&RandomRangeVector(Minimum=<0 -35 70>, Maximum=<0 -11 75>)
     1=&RandomRangeVector(Minimum=<0 11 70>, Maximum=<0 35 75>)
     2=&RandomRangeVector(Minimum=<0 -2 90>, Maximum=<0 2 108>)
     3=&RandomRangeVector(Minimum=<0 -5 -35>, Maximum=<0 5 -40>)
     Index=$Particles.PartitionParticles.Partition
     Num=::NewEnumerator2
    )
    Socket Reader=$System.Socket Reader
    Socket Sampling Mode=::NewEnumerator3
    VelocitySpeedLimit=false
   )
   &ScaleColor(
    Scale Alpha=&FloatFromCurve(
     FloatCurve=DataInterface() {propertyValues {Curve {keys [
         <at 0 0 RCIM_Cubic>
         {arriveTangentWeight=0.1667, interpMode=RCIM_Cubic, leaveTangentWeight=0.1667, tangentMode=RCTM_User, time=0.3, value=1}
         {arriveTangentWeight=0.1667, interpMode=RCIM_Cubic, tangentMode=RCTM_User, time=1, value=0}
        ]}}}
     Scale Curve=0.27
    )
    Scale RGB=&MakeVectorFromLinearColorRGB(Color=&RandomRangeLinearColor(Minimum=<rgba 1 0.5 0 1>, Maximum=<rgba 1 0.6333 0 1>))
    ScaleA=true
    ScaleRGB=true
   )
   &CameraOffset(Camera Offset Amount=&MergeHelper_FloatSwitch(0=10, 1=10, 2=8, 3=?, Index=$Particles.PartitionParticles.Partition, Num=::NewEnumerator1))
  ]
  rendererClasses [&NiagaraSpriteRendererProperties]
  renderers [{rendererClass=&NiagaraSpriteRendererProperties}]
 }
 SplineGlow {
  emitterUpdateScript [
   &EmitterState(
    Inactive Response=::NewEnumerator1
    Life Cycle Mode=::NewEnumerator1
    Loop Behavior=::NewEnumerator1
    Loop Duration Mode=::NewEnumerator1
    UseLoopDelay=false
   )
   &SpawnBurst_Instantaneous(Use Spawn Probability=false, UseLoopCountLimit=false) {module=SpawnBurst_Instantaneous001}
  ]
  particleSpawnScript [&InitializeParticle(Color=<rgba 1 1 1 0.35>, Mesh Scale Mode=::NewEnumerator3, UsePositionOffset=false, Write Material Random=false)]
  particleUpdateScript [
   &ParticleState(Kill Particles When Lifetime Has Elapsed=false)
   &SocketLocation(SamplePosOffset=false, Socket Reader=$System.Socket Reader, Socket Sampling Mode=::NewEnumerator3, VelocitySpeedLimit=false)
   &ScaleColor(
    Scale Alpha=&FN_FrontEndFloatModifier(
     Gameplay=&FloatFromCurve(FloatCurve=DataInterface() {propertyValues {Curve {keys [<at 0 0 RCIM_Cubic>,<at 0.65 1 RCIM_Cubic>]}}}, CurveIndex=$Engine.System.Age)
     Front End=&FloatFromCurve(FloatCurve=DataInterface() {propertyValues {Curve {keys [<at 0.35 0 RCIM_Cubic>,<at 1.1 1 RCIM_Cubic>]}}}, CurveIndex=$Engine.System.Age)
    )
    Scale RGB=&MakeVectorFromLinearColorRGB(Color=<rgba 1.2 0.7 0 1>)
    ScaleA=true
    ScaleRGB=true
   )
  ]
  rendererClasses [&NiagaraMeshRendererProperties]
  renderers [{rendererClass=&NiagaraMeshRendererProperties, rendererIndex=0}]
 }
}
summary {
 systemName=NS_Pickaxe_GoldenSkeleton_IdleGlow
 emitters [
  {emitterName=SplineGlow, rendererClasses [&NiagaraMeshRendererProperties]}
  {emitterName=Decal_Packed, rendererClasses [&NiagaraSpriteRendererProperties]}
 ]
 userVariables [
  User.bIsFrontEnd {type=NiagaraBool}
  User.bIsFrontEndPreview {type=NiagaraBool}
 ]
}
