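# Niagara swing trail for the Golden Skeleton pickaxe, drawn as a short-lived anim-trail ribbon from a single MainTrail emitter.
# FN anim-trail modules spawn, initialize and update particles between two bone or socket indices.
# User.Variant selects a white or near-black color and sets complementary IsDefault/IsVariant emitter flags alongside two ribbon renderers.
# Alpha fades over a curve scaled by System.SwingGeneralFade.
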NiagaraSystem /BRCosmetics/Weapons/FORT_Melee/Pickaxe_TheGoldenSkeletonPickaxe/FX/NS_Pickaxe_GoldenSkeleton_Trail
dependencies {
 usedDataInterfaces [&NiagaraDataInterfaceCurve, &NiagaraDataInterfaceSocketReader]
 usedDynamicInputs []
 usedModules [
  &self:SystemSpawnScript.SystemScriptSource.SystemScriptGraph.NiagaraNodeAssignment_1.SetVariables_#1
  &SystemState
  &self:SystemSpawnScript.SystemScriptSource.SystemScriptGraph.NiagaraNodeAssignment_0.SetVariables_#2
 ]
}
emitters {MainTrail {
  emitterSpawnScript [&FN_RibbonIDOffsetUtility()]
  emitterUpdateScript [
   &EmitterState()
   &FN_SpawnAnimTrail()
   /SetVariables(Emitter.IsVariant:NiagaraBool=&SetBoolByIntComparison(A=&MakeIntFromFloat(Float=$User.Variant), B=1, Comparison Type=::NewEnumerator4)) {isSetParameters=true}
   /SetVariables(Emitter.IsDefault:NiagaraBool=&Boolean_NotOperation(Bool=$Emitter.IsVariant)) {isSetParameters=true}
  ]
  particleSpawnScript [
   &FN_InitializeAnimTrailBoneIndecies(First Bone or Socket Index=1, Second Bone or Socket Index=0)
   &FN_InitializeAnimTrail(
    Color=&MergeHelper_ColorSwitch(
     0=<rgba 1 1 1 1>
     1=<rgba 0.0142 0.0101 0.019 1.7>
     2=?
     3=?
     Index=&MakeIntFromFloat(Float=$User.Variant)
     Num=::NewEnumerator0
    )
    Lifetime=0.25
   )
  ]
  particleUpdateScript [
   &FN_UpdateAnimTrail() {module=FN_UpdateAnimTrail001}
   &ScaleColor(
    Scale Alpha=&FloatFromCurve(FloatCurve=DataInterface() {propertyValues {Curve {keys [<at 0 1 RCIM_Cubic>,<at 1 0 RCIM_Cubic>]}}}, Scale Curve=$System.SwingGeneralFade)
    ScaleA=true
    ScaleRGB=true
   )
  ]
  rendererClasses [&NiagaraRibbonRendererProperties]
  renderers [
   {rendererClass=&NiagaraRibbonRendererProperties}
   {rendererClass=&NiagaraRibbonRendererProperties, rendererIndex=1}
  ]
 }}
summary {
 systemName=NS_Pickaxe_GoldenSkeleton_Trail
 emitters [{emitterName=MainTrail, rendererClasses [&NiagaraRibbonRendererProperties]}]
 userVariables [
  User.RibbonIDOffset {type=NiagaraInt32}
  User.Variant {type=NiagaraFloat, defaultValue=0, description=""}
 ]
}
